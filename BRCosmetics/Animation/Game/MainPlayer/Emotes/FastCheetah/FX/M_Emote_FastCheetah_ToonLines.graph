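# Translucent, unlit, two-sided material for Niagara ribbons, drawing toon-style line streaks in the FastCheetah main-player emote effect.
# Particle color RGB passes through the shared emissive adjustment function to tint the lines.
# A stretched Perlin noise texture pans along U, then a power and step cut it into hard-edged streaks, which are multiplied by particle alpha as a mask.
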Material /BRCosmetics/Animation/Game/MainPlayer/Emotes/FastCheetah/FX/M_Emote_FastCheetah_ToonLines
graph {
 properties {
  bUsedWithNiagaraRibbons=true
  blendMode=BLEND_Translucent
  pixelDepthOffsetMode=PDOM_Legacy
  shadingModel=MSM_Unlit
  twoSided=true
  usedShadingModels=Unlit
 }
 nodes [
  n1 = ParticleColor()
  MaterialFunctionCall(Emissive=@n1.RGB) {materialFunction=&MF_FX_EmissiveAdjustments}
  n2 = TextureCoordinate()
  n3 = Constant2Vector() {g=6.5, r=0.2}
  n4 = TextureSample(UVs=Panner(Coordinate=Multiply(A=@n2, B=@n3)) {speedX=0.6}) {samplerType=SAMPLERTYPE_LinearGrayscale, texture=&T_SimplePerlin_Greyscale}
  Multiply(A=Step(X=Power(Base=@n4.R) {constExponent=3}) {constX=0.2, constY=0.4}, B=@n1.A)
 ]
}
