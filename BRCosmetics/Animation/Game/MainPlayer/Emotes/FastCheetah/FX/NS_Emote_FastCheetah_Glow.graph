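# Niagara effect for the FastCheetah emote that adds a flickering cyan glow to the skeletal mesh passed in through User.SkeletalMesh.
# The Empty emitter burst-spawns mesh particles that resample their mesh location every update.
# The Glow emitter places sprites at spawn and draws them through two differently configured sprite renderers.
# Both fade in, pulse alpha via curves over emitter loop age, and keep particles alive past their lifetime.
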NiagaraSystem /BRCosmetics/Animation/Game/MainPlayer/Emotes/FastCheetah/FX/NS_Emote_FastCheetah_Glow
dependencies {
 usedDataInterfaces [&NiagaraDataInterfaceCurve, &NiagaraDataInterfaceColorCurve]
 usedDynamicInputs []
 usedModules [&SystemState]
}
emitters {
 Empty {
  emitterUpdateScript [
   &EmitterState()
   &SpawnBurst_Instantaneous(Use Spawn Probability=false, UseLoopCountLimit=false)
  ]
  particleSpawnScript [&InitializeParticle(
    Color=<rgba 0.3008 0.9602 1 0.2>
    Lifetime=5
    Mesh Scale=<0.21 0.21 0.5>
    Mesh Scale Mode=::NewEnumerator4
    UsePositionOffset=false
    Write Material Random=true
   )]
  particleUpdateScript [
   &ParticleState(Kill Particles When Lifetime Has Elapsed=false)
   &SkeletalMeshLocation(
    Bone Sampling Mode=::NewEnumerator4
    MaskSpawnGroup=false
    Orientation Sampling=::NewEnumerator1
    Sampled Position Offset=<-1.5 0 0>
    Skeletal Mesh=$User.SkeletalMesh
   )
   &ScaleColor(
    Scale Alpha=&FloatFromCurve(FloatCurve=DataInterface() {propertyValues {Curve {keys [
         {arriveTangent=4.1785, interpMode=RCIM_Cubic, leaveTangent=4.1785, tangentMode=RCTM_User, time=0, value=0}
         {arriveTangent=0.0224, interpMode=RCIM_Cubic, leaveTangent=0.0224, tangentMode=RCTM_User, time=0.3, value=0.95}
        ]}}})
    ScaleA=true
    ScaleRGB=true
   )
   &ScaleColor(
    Curve Index=$Emitter.NormalizedLoopAge
    Linear Color Curve=DataInterface() {propertyValues {
      AlphaCurve {keys [<at 0 0.2>,<at 0.092 1>,<at 0.1989 0.25>,<at 0.3165 0.6>,<at 0.4448 0.6>,<at 0.5389 1>,<at 0.633 0.4>,<at 0.725 1>,<at 0.8255 0.3>,<at 0.9239 0.7>,<at 1 0.2>]}
      BlueCurve {keys [<at 0 1>,<at 1 1>]}
      GreenCurve {keys [<at 0 1>,<at 1 1>]}
      RedCurve {keys [<at 0 1>,<at 1 1>]}
     }}
    Scale Mode=::NewEnumerator2
   ) {module=ScaleColor001}
  ]
  rendererClasses [&NiagaraMeshRendererProperties]
  renderers [{rendererClass=&NiagaraMeshRendererProperties, rendererIndex=0}]
 }
 Glow {
  emitterUpdateScript [
   &EmitterState()
   &SpawnBurst_Instantaneous(Use Spawn Probability=false, UseLoopCountLimit=false)
  ]
  particleSpawnScript [
   &InitializeParticle(
    Color=<rgba 0.4166 1 0.9901 0.1>
    Lifetime=0.35
    Sprite Size=<75 10>
    Sprite Size Mode=::NewEnumerator4
    UsePositionOffset=false
    Write Material Random=true
   )
   &SkeletalMeshLocation(
    MaskSpawnGroup=false
    Mesh Sampling Type=::NewEnumerator5
    Sampled Position Offset=<-12 0 0>
    Skeletal Mesh=$User.SkeletalMesh
    Socket Sampling Mode=::NewEnumerator5
   )
   /SetVariables(Particles.SpriteAlignment=<0 0 1>) {isSetParameters=true}
   /SetVariables(
    Particles.Render2SpriteAlignment:Vector3f=<1 0 0>
    Particles.Render2SpriteFacing:Vector3f=<0 0 1>
    Particles.Render2SpriteSize:Vector2f=<75 75>
   ) {isSetParameters=true}
   &DynamicMaterialParameters(
    Index 0 Param 1=5
    Module.Index0Float0Write=true
    Module.Index0Float1Write=false
    Module.Index0Float2Write=false
    Module.Index0Float3Write=false
   )
  ]
  particleUpdateScript [
   &ParticleState(Kill Particles When Lifetime Has Elapsed=false)
   &ScaleColor(
    Scale Alpha=&FloatFromCurve(FloatCurve=DataInterface() {propertyValues {Curve {keys [
         {arriveTangent=4.1785, interpMode=RCIM_Cubic, leaveTangent=4.1785, tangentMode=RCTM_User, time=0, value=0}
         {arriveTangent=0.0224, interpMode=RCIM_Cubic, leaveTangent=0.0224, tangentMode=RCTM_User, time=0.6, value=1}
        ]}}})
    ScaleA=true
    ScaleRGB=true
   )
   &ScaleColor(
    Curve Index=$Emitter.NormalizedLoopAge
    Linear Color Curve=DataInterface() {propertyValues {
      AlphaCurve {keys [<at 0 0>,<at 0.092 1>,<at 0.1989 0.25>,<at 0.3165 0.6>,<at 0.4448 0>,<at 0.5389 1>,<at 0.633 0.4>,<at 0.725 1>,<at 0.8255 0>,<at 0.9239 0.7>,<at 1 0>]}
      BlueCurve {keys [<at 0 1>,<at 1 1>]}
      GreenCurve {keys [<at 0 1>,<at 1 1>]}
      RedCurve {keys [<at 0 1>,<at 1 1>]}
     }}
    Scale Mode=::NewEnumerator2
   ) {module=ScaleColor001}
  ]
  rendererClasses [&NiagaraSpriteRendererProperties]
  renderers [
   {rendererClass=&NiagaraSpriteRendererProperties}
   {rendererClass=&NiagaraSpriteRendererProperties, rendererIndex=1}
  ]
 }
}
summary {
 systemName=NS_Emote_FastCheetah_Glow
 emitters [
  {emitterName=Glow, rendererClasses [&NiagaraSpriteRendererProperties]}
  {emitterName=Empty, rendererClasses [&NiagaraMeshRendererProperties]}
 ]
 userVariables [User.SkeletalMesh {
   type=NiagaraDataInterfaceSkeletalMesh
   defaultValue {
    struct=NiagaraExt_VariableValue_DataInterface
    value {dataInterface=&self:NiagaraDataInterfaceSkeletalMesh_0}
   }
  }]
}
