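# Niagara burst effect for the punch explosion in the FastCheetah emote.
# Short-lived sprite bursts form a flare with a light at System.HitOffset, a star flash, and a smoke puff animated by feeding normalized age into its material.
# Star and smoke colors tint by time of day, flare and star camera offsets switch for front end, and the star adds distance scaling and culling.
# A GPU sand emitter is present but disabled.
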NiagaraSystem /BRCosmetics/Animation/Game/MainPlayer/Emotes/FastCheetah/FX/NS_Emote_FastCheetah_PunchExplode
dependencies {
 usedDataInterfaces [&NiagaraDataInterfaceCurve, &NiagaraDataInterfaceColorCurve]
 usedDynamicInputs [
  &Multiply_Float
  &FNTimeOfDayScalar
  &Vector2DFromFloat
  &SimulationPosition
 ]
 usedModules [&SystemState]
}
emitters {
 Flash_Flare001 {
  emitterUpdateScript [
   &EmitterState()
   &SpawnBurst_Instantaneous(Spawn Count=1, Use Spawn Probability=false)
  ]
  particleSpawnScript [
   &InitializeParticle(
    Color=&HueShiftLinearColor(Hue=-0.02, Saturation=1, Value=0.2, Linear Color=$Particles.Color)
    Lifetime=0.3
    Mass Mode=::NewEnumerator1
    Sprite Size Mode=::NewEnumerator3
    Uniform Sprite Size=150
    UsePositionOffset=false
    Write Material Random=true
   ) {module=InitializeParticle001}
   &CameraOffset(Camera Offset Amount=&FN_FrontEndFloatModifier(Gameplay=50, Front End=0))
   &DynamicMaterialParameters(
    Index 0 Param 1=0.2
    Index 0 Param 2=1
    Module.Param0WriteEnabled=true
    Module.Param1WriteEnabled=true
    Module.Param2WriteEnabled=true
    Module.Param3WriteEnabled=true
   )
   &SystemLocation(MaskSpawnGroup=false, Offset=$System.HitOffset)
  ]
  particleUpdateScript [
   &ParticleState()
   &ScaleColor(
    Scale Alpha=&FloatFromCurve(
     FloatCurve=DataInterface() {propertyValues {Curve {keys [<at 0 0.75>, {interpMode=RCIM_Cubic, leaveTangent=-2.1292, tangentMode=RCTM_Break, time=0.2, value=1}, <at 1 0>]}}}
     Scale Curve=0.7
    )
    Scale RGB=&VectorFromFloat(Value=1)
    ScaleA=true
    ScaleRGB=true
   )
   &FN_Light_v2(Light Exponent=3, Light Radius=150, LightEnabled=true, Scale Color=1)
   &CameraOffset(Camera Offset Amount=5) {module=CameraOffset001}
  ]
  rendererClasses [&NiagaraSpriteRendererProperties]
  renderers [{rendererClass=&NiagaraSpriteRendererProperties}]
 }
 Flash_Star {
  emitterUpdateScript [
   &EmitterState(
    Enable Distance Culling=true
    Max Distance Response=::KillImmediately
    Module.bUseMaxDistance=true
    Module.bUseMinDistance=false
    Scalability Mode=::NewEnumerator1
   )
   &SpawnBurst_Instantaneous(Use Spawn Probability=false, UseLoopCountLimit=false)
  ]
  particleSpawnScript [
   &InitializeParticle(
    Lifetime=0.15
    Sprite Rotation Mode=::NewEnumerator2
    Sprite Size Mode=::NewEnumerator3
    Uniform Sprite Size=&UniformRangedFloat(Minimum=80, Maximum=110)
    UsePositionOffset=false
    Write Material Random=true
   )
   &FN_PerParticleLOD(Enable SpriteSize Distance Scaling=true, Sprite Size Scale Far=<2.5 2.5>)
   &CameraOffset(Camera Offset Amount=&FN_FrontEndFloatModifier(Gameplay=25, Front End=5))
   /SetVariables(Particles.RandomFloat=&UniformRangedFloat(Minimum=0, Maximum=1)) {isSetParameters=true}
   /SetVariables(
    Particles.SubImageIndex=&RandomRangeFloat(Minimum=0, Maximum=3)
    Particles.SubImageIndexAlt:NiagaraFloat=&RandomRangeFloat(Minimum=0, Maximum=3)
   ) {isSetParameters=true}
   &Color(Color=&MultiplyLinearColor(
     Linear Color=$Particles.Initial.Color
     Scale Factor=&MakeVector4_FromLinearColor(LinearColor=&FN_Select_Color_from_Time_Of_Day(Morning=<rgba 1 0.7438 0.5801 1>, Day=<rgba 1 1 1 1>, Evening=<rgba 1 1 1 1>, Night=<rgba 0.1039 0.211 0.3594 1>))
    ))
  ]
  particleUpdateScript [
   &ParticleState()
   &ScaleColor(
    Linear Color Curve=DataInterface() {propertyValues {
      AlphaCurve {keys [<at 0 1>,<at 0.8992 0>]}
      BlueCurve {keys [<at 0.3817 6.4583>,<at 0.4644 1.7396>,<at 0.899 0.107>]}
      GreenCurve {keys [<at 0.3817 6.4583>,<at 0.4644 1.9212>,<at 0.899 0.305>]}
      RedCurve {keys [<at 0.3817 8>,<at 0.4644 2>,<at 0.899 0.3854>]}
     }}
    Scale Mode=::NewEnumerator2
   )
   &ScaleSpriteSize(Uniform Curve Sprite Scale=DataInterface() {propertyValues {Curve {keys [{arriveTangent=1.9953, interpMode=RCIM_Cubic, leaveTangent=1.9953, tangentMode=RCTM_User, time=0, value=0.3}, <at 1 1 RCIM_Cubic>]}}}) {module=ScaleSpriteSize001}
  ]
  rendererClasses [&NiagaraSpriteRendererProperties]
  renderers [
   {rendererClass=&NiagaraSpriteRendererProperties}
   {rendererClass=&NiagaraSpriteRendererProperties, rendererIndex=1}
  ]
 }
 SmokePuff {
  emitterUpdateScript [
   &EmitterState()
   &SpawnBurst_Instantaneous(Use Spawn Probability=false, UseLoopCountLimit=false)
  ]
  particleSpawnScript [
   &InitializeParticle(
    Color=<rgba 0.8599 0.9019 1 1>
    Lifetime=0.25
    Sprite Size Mode=::NewEnumerator3
    Uniform Sprite Size=90
    UsePositionOffset=false
    Write Material Random=true
   )
   &Color(Color=&MultiplyLinearColor(
     Linear Color=$Particles.Initial.Color
     Scale Factor=&MakeVector4_FromLinearColor(LinearColor=&FN_Select_Color_from_Time_Of_Day(Morning=<rgba 1 0.7438 0.5801 1>, Day=<rgba 1 1 1 1>, Evening=<rgba 1 1 1 1>, Night=<rgba 0.4556 0.5267 0.6823 1>))
    ))
  ]
  particleUpdateScript [
   &ParticleState()
   &DynamicMaterialParameters(
    Index 0 Param 1=$Particles.NormalizedAge
    Module.Index0Param0Write=true
    Module.Index0Param1Write=true
    Module.Index0Param2Write=true
    Module.Index0Param3Write=true
   )
  ]
  rendererClasses [&NiagaraSpriteRendererProperties]
  renderers [{rendererClass=&NiagaraSpriteRendererProperties}]
 }
 sand {
  bEnabled=false
  simTarget=GPUComputeSim
  emitterUpdateScript [
   &EmitterState()
   &SpawnBurst_Instantaneous(Spawn Count=5, Use Spawn Probability=false, UseLoopCountLimit=false)
   &SpawnRate(SpawnRate=25, Use Spawn Probability=false)
  ]
  particleSpawnScript [
   &InitializeParticle(
    Color Minimum=<rgba 0.4427 0.4427 0.4427 1>
    Color Mode=::NewEnumerator2
    Lifetime Max=0.6
    Lifetime Min=0.25
    Lifetime Mode=::NewEnumerator1
    Sprite Rotation Mode=::NewEnumerator2
    Sprite Size Mode=::NewEnumerator1
    Uniform Sprite Size Max=4
    Uniform Sprite Size Min=2
    UsePositionOffset=false
    Write Material Random=true
   )
   &ShapeLocation(MaskSpawnGroup=false, Override Seed=false, Sphere Radius=6, UseOwnerScale=true)
   /Game/Effects/Niagara/Modules/Sprite/FN_SpriteUV(End Frame=3)
  ]
  particleUpdateScript [
   &ParticleState()
   &GravityForce(Gravity=&RandomRangeVector(Minimum=<0 0 -450>, Maximum=<1 1 -980>))
   &SolveForcesAndVelocity(Clamp Velocity=false, Limit Acceleration=false)
  ]
  rendererClasses [&NiagaraSpriteRendererProperties]
  renderers [{rendererClass=&NiagaraSpriteRendererProperties}]
 }
}
summary {
 systemName=NS_Emote_FastCheetah_PunchExplode
 emitters [
  {emitterName=SmokePuff, rendererClasses [&NiagaraSpriteRendererProperties]}
  {emitterName=Flash_Star, rendererClasses [&NiagaraSpriteRendererProperties]}
  {emitterName=Flash_Flare001, rendererClasses [&NiagaraSpriteRendererProperties]}
  {bEnabled=false, emitterName=sand, simTarget=GPUComputeSim, rendererClasses [&NiagaraSpriteRendererProperties]}
 ]
 userVariables [
  User.bIsFrontEnd {type=NiagaraBool}
  User.bIsFrontEndPreview {type=NiagaraBool}
  User.RibbonIDOffset {type=NiagaraInt32}
  User.SkeletalMesh {
   type=NiagaraDataInterfaceSkeletalMesh
   defaultValue {
    struct=NiagaraExt_VariableValue_DataInterface
    value {dataInterface=&self:NiagaraDataInterfaceSkeletalMesh_0}
   }
  }
 ]
}
