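# Niagara system for the FastCheetah emote that draws short-lived speed-line ribbon trails from a skeletal mesh passed in as a user parameter.
# Two emitters share FN_AnimTrail modules to pick bone indices, spawn trails and update ribbons against that mesh.
# The wider, light trail is tinted by time of day and fades out through an alpha curve.
# The narrower black trail skips tinting and ramps alpha up early in its life.
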NiagaraSystem /BRCosmetics/Animation/Game/MainPlayer/Emotes/FastCheetah/FX/NS_Emote_FastCheetah_SpeedLines
dependencies {
 usedDataInterfaces [&NiagaraDataInterfaceColorCurve, &NiagaraDataInterfaceCurve]
 usedDynamicInputs []
 usedModules [&SystemState]
}
emitters {
 FN_AnimTrail {
  emitterSpawnScript [&FN_RibbonIDOffsetUtility()]
  emitterUpdateScript [
   &EmitterState()
   &FN_SpawnAnimTrail()
  ]
  particleSpawnScript [
   &FN_InitializeAnimTrailBoneIndecies()
   &FN_InitializeAnimTrail(AnimTrailWidth=0.7, Color=<rgba 0.5956 0.6011 0.5865 0.6>, Lifetime=0.1)
   &Color(Color=&MultiplyLinearColor(
     Linear Color=$Particles.Initial.Color
     Scale Factor=&MakeVector4_FromLinearColor(LinearColor=&FN_Select_Color_from_Time_Of_Day(Morning=<rgba 0.903 0.75 0.6359 0.5>, Day=<rgba 0.8747 0.8747 0.8747 0.5>, Evening=<rgba 1 1 1 0.5>, Night=<rgba 0.1555 0.225 0.5775 0.5>))
    ))
  ]
  particleUpdateScript [
   &SolveForcesAndVelocity(Clamp Velocity=false, Limit Acceleration=false)
   &FN_UpdateAnimTrail(Locked Particle RibbonWidth=1, SkeletalMesh=$User.SkeletalMesh) {module=FN_UpdateAnimTrail001}
   &ScaleColor(
    Scale Alpha=&FloatFromCurve(FloatCurve=DataInterface() {propertyValues {Curve {keys [
         <at 0 0.5>
         {interpMode=RCIM_Cubic, leaveTangent=-0.5997, tangentMode=RCTM_Break, time=0.1, value=1}
         {arriveTangent=-12.5603, tangentMode=RCTM_User, time=0.4, value=0}
        ]}}})
    ScaleA=true
    ScaleRGB=true
   )
  ]
  rendererClasses [&NiagaraRibbonRendererProperties]
  renderers [{rendererClass=&NiagaraRibbonRendererProperties}]
 }
 FN_AnimTrail001 {
  emitterSpawnScript [&FN_RibbonIDOffsetUtility()]
  emitterUpdateScript [
   &EmitterState()
   &FN_SpawnAnimTrail()
  ]
  particleSpawnScript [
   &FN_InitializeAnimTrailBoneIndecies()
   &FN_InitializeAnimTrail(AnimTrailWidth=0.4, Color=<rgba 0 0 0 0.4>, Lifetime=0.1)
  ]
  particleUpdateScript [
   &SolveForcesAndVelocity(Clamp Velocity=false, Limit Acceleration=false)
   &FN_UpdateAnimTrail(Locked Particle RibbonWidth=1, SkeletalMesh=$User.SkeletalMesh) {module=FN_UpdateAnimTrail001}
   &ScaleColor(
    Scale Alpha=&FloatFromCurve(FloatCurve=DataInterface() {propertyValues {Curve {keys [<at 0 0.25>, {interpMode=RCIM_Cubic, leaveTangent=0.2143, tangentMode=RCTM_Break, time=0.1, value=1}]}}})
    ScaleA=true
    ScaleRGB=true
   )
  ]
  rendererClasses [&NiagaraRibbonRendererProperties]
  renderers [{rendererClass=&NiagaraRibbonRendererProperties}]
 }
}
summary {
 systemName=NS_Emote_FastCheetah_SpeedLines
 emitters [
  {emitterName=FN_AnimTrail, rendererClasses [&NiagaraRibbonRendererProperties]}
  {emitterName=FN_AnimTrail001, rendererClasses [&NiagaraRibbonRendererProperties]}
 ]
 userVariables [
  User.RibbonIDOffset {type=NiagaraInt32}
  User.SkeletalMesh {
   type=NiagaraDataInterfaceSkeletalMesh
   defaultValue {
    struct=NiagaraExt_VariableValue_DataInterface
    value {dataInterface=&self:NiagaraDataInterfaceSkeletalMesh_0}
   }
  }
 ]
}
